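# Sound cue played when interacting with the Ancient Weapon spawner, using item-interact attenuation and the loot-collect sound class.
# A mixer layers a main interact wave, a delayed interact layer, an equip sound and randomized catch and ADS-in zap variants, the latter delayed.
# Most layers are pitched up, and a root modulator adds slight pitch variation.
# The equip randomizer's three slots all name the same TEMP wave, so it adds no variety.
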SoundCue /AncientWeapon/Environmental/Sound/AncientWeapon_Spawner_Interact_Cue
root=Modulator_0
nodes [
 {id=Delay_0, children [Modulator_2]}
 {id=Delay_1, children [WavePlayer_12], properties {delayMax=0.1, delayMin=0.1}}
 {id=Delay_2, children [Modulator_3], properties {delayMax=0.3, delayMin=0.3}}
 {id=Mixer_0, children [Modulator_4,Delay_1,Modulator_1,Delay_0,Delay_2], properties {inputVolume [0.8,0.6,0.6,1,2]}}
 {id=Modulator_0, children [Mixer_0], properties {pitchMax=1.1, pitchMin=1}}
 {id=Modulator_1, children [Random_2], properties {pitchMax=1.2, pitchMin=1.2}}
 {id=Modulator_2, children [Random_0], properties {pitchMax=1.2, pitchMin=1.2}}
 {id=Modulator_3, children [Random_1], properties {pitchMax=1.2, pitchMin=1.2}}
 {id=Modulator_4, children [WavePlayer_0], properties {pitchMax=1.2, pitchMin=1.2}}
 {id=Random_0, children [WavePlayer_4,WavePlayer_5,WavePlayer_6], properties {weights [1,1,1]}}
 {id=Random_1, children [WavePlayer_7, WavePlayer_8, WavePlayer_10], properties {weights [1,1,1]}}
 {id=Random_2, children [WavePlayer_9,WavePlayer_1,WavePlayer_2], properties {weights [1,1,1]}}
 {id=WavePlayer_0, properties {soundWaveAssetPtr=&/AncientWeapon_Spawner_Interact_01}}
 {id=WavePlayer_1, properties {soundWaveAssetPtr=&/AncientWeapon_TEMP_EquipB_01}}
 {id=WavePlayer_10, properties {soundWaveAssetPtr=/AncientWeapon/Assets/Audio/Ancient_Weapon_ADS_In_Zaps_02}}
 {id=WavePlayer_12, properties {soundWaveAssetPtr=&/AncientWeapon_Spawner_Interact_Layer_01}}
 {id=WavePlayer_2, properties {soundWaveAssetPtr=&AncientWeapon_TEMP_EquipB_01}}
 {id=WavePlayer_4, properties {soundWaveAssetPtr=/AncientWeapon/Assets/Audio/Ancient_Weapon_Catch_Zaps_02}}
 {id=WavePlayer_5, properties {soundWaveAssetPtr=/AncientWeapon/Assets/Audio/Ancient_Weapon_Catch_Zaps_03}}
 {id=WavePlayer_6, properties {soundWaveAssetPtr=/AncientWeapon/Assets/Audio/Ancient_Weapon_Catch_Zaps_04}}
 {id=WavePlayer_7, properties {soundWaveAssetPtr=/AncientWeapon/Assets/Audio/Ancient_Weapon_ADS_In_Zaps_03}}
 {id=WavePlayer_8, properties {soundWaveAssetPtr=/AncientWeapon/Assets/Audio/Ancient_Weapon_ADS_In_Zaps_01}}
 {id=WavePlayer_9, properties {soundWaveAssetPtr=&AncientWeapon_TEMP_EquipB_01}}
]
properties {
 attenuationSettings=/Game/Sounds/Attenuation/Fort_Item_Interact_Mono_Att
 maxDistance=1500
 soundClassObject=&Fort_Loot_Collect
 volumeMultiplier=0.9
 concurrencySet [null]
}
