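# Sound cue for the start of an interaction with an Ancient Weapon spawner.
# A mixer layers a slightly boosted spawner grab sound with one of three equally weighted gear-search foley waves picked at random.
# A root modulator adds slight upward pitch variation to the mix.
# It plays at reduced volume through the loot-collect sound class with mono item-interact attenuation.
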SoundCue /AncientWeapon/Environmental/Sound/AncientWeapon_Spawner_Start_Interact_Cue
root=Modulator_0
nodes [
 {id=Mixer_0, children [WavePlayer_1,Random_12], properties {inputVolume [1.2,1]}}
 {id=Modulator_0, children [Mixer_0], properties {pitchMax=1.1, pitchMin=1}}
 {id=Random_12, children [WavePlayer_58, WavePlayer_59, WavePlayer_67], properties {weights [1,1,1]}}
 {id=WavePlayer_1, properties {soundWaveAssetPtr=&/AncientWeapon_Spawner_Grab_01}}
 {id=WavePlayer_58, properties {soundWaveAssetPtr=/Game/Sounds/Foley_Loot/Containers/Misc/Prop_Gear_Search_05}}
 {id=WavePlayer_59, properties {soundWaveAssetPtr=/Game/Sounds/Foley_Loot/Containers/Misc/Prop_Gear_Search_06}}
 {id=WavePlayer_67, properties {soundWaveAssetPtr=/Game/Sounds/Foley_Loot/Containers/Misc/Prop_Gear_Search_01}}
]
properties {
 attenuationSettings=/Game/Sounds/Attenuation/Fort_Item_Interact_Mono_Att
 maxDistance=1500
 soundClassObject=&Fort_Loot_Collect
 volumeMultiplier=0.4
 concurrencySet [null]
}
